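# Niagara system for the reactive effect on the WavySignalBanjo backpack, combining egg glows with lightning, all placed relative to a shared system offset.
# EggGlows bursts six sprites in a ring, each triple-rendered with red and purple layers that flare then fade, alongside EggRedLight's red component-rendered pulse.
# Lightning continuously spawns random mesh arcs on a cylinder swept by age-driven sine and cosine, with front-end-aware material parameters, while LightningLight adds short yellow component flickers.
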NiagaraSystem /BRCosmetics/Accessories/FORT_Backpacks/Backpack_M_MED_WavySignalBanjo/FX/NS_Backpack_WavySignalBanjo_Reactive
dependencies {
 usedDataInterfaces [
  &NiagaraDataInterfaceCurve
  &NiagaraDataInterfaceArrayFloat
  &NiagaraDataInterfaceMeshRendererInfo
 ]
 usedDynamicInputs [
  &SimulationPosition
  &Add_Integer
  &Multiply_Int
  &Subtract_Vector
 ]
 usedModules [
  &self:SystemSpawnScript.SystemScriptSource.SystemScriptGraph.NiagaraNodeAssignment_1.SetVariables_#1
  &SystemState
 ]
}
emitters {
 EggGlows {
  emitterUpdateScript [
   &EmitterState()
   &SpawnBurst_Instantaneous(Spawn Count=6, Use Spawn Probability=false, UseLoopCountLimit=false)
  ]
  particleSpawnScript [&InitializeParticle(
    Color Mode=::NewEnumerator0
    Lifetime=3
    Position Offset=&Add_Vector(
     A=$System.Offset
     B=/Niagara/DynamicInputs/Vectors/RotateVector(
      Vector To Rotate:Vector3f=<0 0 10>
      Yaw:NiagaraFloat=0
      Pitch:NiagaraFloat=0
      Roll:NiagaraFloat=&SelectFloatFromArray(
       Direct Array Index=&ReturnParticleIDAttributes()
       Float Selection Array=DataInterface() {propertyValues {FloatData [7,67,127,187,247,307]}}
      )
      Rotation Rate:NiagaraFloat=1
     )
    )
    Position Offset Coordinate Space=::Local
    Sprite Size=&Multiply_Vector2D_ByFloat(Float=0.07, Vector2D=<30 20>)
    Sprite Size Mode=::NewEnumerator4
    UsePositionOffset=true
    Write Material Random=true
   )]
  particleUpdateScript [
   /SetVariables(
    Particles.SpriteHiLiColor:LinearColor=&Multiply_LinearColorByFloat(
     Linear Color=<rgba 0.2869 0 1 1.5>
     Float=&FloatFromCurve(FloatCurve=DataInterface() {propertyValues {Curve {keys [
          {interpMode=RCIM_Cubic, tangentMode=RCTM_User, time=0, value=0}
          {interpMode=RCIM_Cubic, tangentMode=RCTM_User, time=0.1, value=1}
          {arriveTangent=-3.771, tangentMode=RCTM_User, time=1, value=0}
         ]}}})
    )
    Particles.SpriteHiLiSize:Vector2f=&Multiply_Vector2D_ByFloat(Float=0.12, Vector2D=<30 20>)
    Particles.SpriteSmallColor:LinearColor=&Multiply_LinearColorByFloat(
     Linear Color=<rgba 1 0 0 2>
     Float=&FloatFromCurve(FloatCurve=DataInterface() {propertyValues {Curve {keys [
          {interpMode=RCIM_Cubic, tangentMode=RCTM_User, time=0, value=0}
          {interpMode=RCIM_Cubic, tangentMode=RCTM_User, time=0.1, value=1}
          {arriveTangent=-3.771, tangentMode=RCTM_User, time=1, value=0}
         ]}}})
    )
    Particles.SpriteSmallSize:Vector2f=&Multiply_Vector2D_ByFloat(Float=0.2, Vector2D=<30 20>)
   ) {isSetParameters=true}
   &Color(Color=&Multiply_LinearColorByFloat(
     Linear Color=<rgba 1 0 0.0001 2.0174>
     Float=&FloatFromCurve(FloatCurve=DataInterface() {propertyValues {Curve {keys [
          {interpMode=RCIM_Cubic, tangentMode=RCTM_User, time=0, value=0}
          {interpMode=RCIM_Cubic, tangentMode=RCTM_User, time=0.1, value=1}
          {arriveTangent=-3.8765, tangentMode=RCTM_User, time=1, value=0}
         ]}}})
    ))
   &CameraOffset(Camera Offset Amount=5)
   &ParticleState()
  ]
  rendererClasses [&NiagaraSpriteRendererProperties]
  renderers [
   {rendererClass=&NiagaraSpriteRendererProperties}
   {rendererClass=&NiagaraSpriteRendererProperties, rendererIndex=1}
   {rendererClass=&NiagaraSpriteRendererProperties, rendererIndex=2}
  ]
 }
 EggRedLight {
  emitterUpdateScript [
   &EmitterState()
   &SpawnBurst_Instantaneous(Use Spawn Probability=false, UseLoopCountLimit=false)
  ]
  particleSpawnScript [&InitializeParticle(
    Color=<rgba 5 5 5 1>
    Lifetime=3
    Position Offset=&Add_Vector(A=$System.Offset, B=<0 -5 0>)
    UsePositionOffset=true
    Write Material Random=true
   )]
  particleUpdateScript [
   &ParticleState()
   &ScaleColor(
    Scale RGB=&Multiply_VectorByFloat(
     Vector=<1 0 0>
     Float=&FloatFromCurve(FloatCurve=DataInterface() {propertyValues {Curve {keys [
          {interpMode=RCIM_Cubic, tangentMode=RCTM_User, time=0, value=0}
          {interpMode=RCIM_Cubic, tangentMode=RCTM_User, time=0.1, value=1}
          {arriveTangent=0.0152, interpMode=RCIM_Cubic, leaveTangent=0.0152, tangentMode=RCTM_User, time=0.5, value=1}
          {arriveTangent=-4.0375, tangentMode=RCTM_User, time=1, value=0}
         ]}}})
    )
    ScaleA=true
    ScaleRGB=true
   )
  ]
  rendererClasses [&NiagaraComponentRendererProperties]
  renderers [{rendererClass=&NiagaraComponentRendererProperties, rendererIndex=0}]
 }
 Lightning {
  emitterUpdateScript [
   &EmitterState()
   &SpawnRate(SpawnRate=30, Use Spawn Probability=true)
  ]
  particleSpawnScript [
   &InitializeParticle(
    Color=<rgba 5 4.7492 4.1927 1>
    Lifetime=&RandomRangeFloat(Minimum=0.1, Maximum=0.5)
    UsePositionOffset=false
    Write Material Random=true
   )
   /SetVariables(Particles.bFlipMesh:NiagaraBool=&RandomBool(Randomness Mode=::NewEnumerator2, Random Seed=?)) {isSetParameters=true}
   /SetVariables(Particles.MeshIndex=&RandomRangeInt(Minimum=0, Maximum=2)) {isSetParameters=true}
   /SetVariables(Particles.Scale=&Multiply_VectorByFloat(Vector=<1 0.5 1>, Float=&RandomRangeFloat(Minimum=0.15, Maximum=0.25))) {isSetParameters=true}
   &ShapeLocation(
    Cylinder Distribution=::NewEnumerator2
    Cylinder Height=20
    Cylinder Radius=9
    MaskSpawnGroup=false
    Offset=$System.Offset
    Offset Mode=::NewEnumerator0
    Rotation Angle Type:ENiagara_AngleInput=::NewEnumerator0
    Rotation Mode=::NewEnumerator2
    Shape Primitive=::NewEnumerator1
    U Position=&Sine(Normalized Angle=$Emitter.Age, Period=1.8, Scale=1, Bias=0)
    UseOwnerScale=true
    V Position=&Cosine(Normalized Angle=$Emitter.Age, Period=1.4, Scale=1, Bias=0)
    Yaw / Pitch / Roll=<0 90 0>
   )
   &InitialMeshOrientation(
    Rotation=&MakeVector(
     X=0
     Y=0
     Z=&MakeCustomFloatFromBool(Bool=&RandomBool(), True Float=1, False Float=0)
    )
    Use Rotation Vector=true
   )
   /SetVariables(Particles.glowcolor1:LinearColor=<rgba 2.5 1.2627 0.0553 0.3217>, Particles.glowcolor2:LinearColor=<rgba 1.5 0.6344 0 0.3217>) {isSetParameters=true}
  ]
  particleUpdateScript [
   &DynamicMaterialParameters(
    Index 0 Param 1=&FN_FrontEndFloatModifier(Gameplay=2, Front End=2.5)
    Index 0 Param 2=&FloatFromCurve(
     FloatCurve=DataInterface() {propertyValues {Curve {
        defaultValue=1
        keys [
         {arriveTangent=9.0725, interpMode=RCIM_Cubic, leaveTangent=9.0725, tangentMode=RCTM_User, time=0, value=0}
         {arriveTangent=-2.5249, interpMode=RCIM_Cubic, leaveTangent=-2.5249, tangentMode=RCTM_User, time=0.1453, value=1.1569}
         {arriveTangent=-2.6402, interpMode=RCIM_Cubic, leaveTangent=-2.6402, time=0.289, value=0.0335}
         {arriveTangent=-2.1502, interpMode=RCIM_Cubic, leaveTangent=-2.1502, tangentMode=RCTM_User, time=0.4576, value=0.3325}
         {arriveTangent=-0.0405, interpMode=RCIM_Cubic, leaveTangent=-0.0405, tangentMode=RCTM_User, time=1, value=0}
        ]
       }}}
     Scale Curve=0.8
    )
    Index 0 Param 3=&Add_Float(
     A=&RandomRangeFloat(Minimum=0, Maximum=10)
     B=&Multiply_Float(
      A=&MakeCustomFloatFromBool(Bool=&RandomBool(), True Float=1, False Float=-1)
      B=&Multiply_Float(A=$Particles.NormalizedAge, B=0.2)
     )
    )
    Index 1 Param 1=0.02
    Index 1 Param 3=&FloatFromCurve(FloatCurve=DataInterface() {propertyValues {Curve {keys [<at 0 1>,<at 1 0>]}}})
    Module.Index0Param0Write=true
    Module.Index0Param1Write=true
    Module.Index0Param2Write=true
    Module.Index0Param3Write=false
    Module.Index1Param0Write=true
    Module.Index1Param1Write=true
    Module.Index1Param2Write=true
    Module.Index1Param3Write=false
    Module.Index2Param0Write=true
    Module.Index2Param1Write=true
    Module.Index2Param2Write=true
    Module.Index2Param3Write=true
    Write Parameter Index 1=true
    Write Parameter Index 2=true
   )
   &ParticleState()
   &ScaleColor(
    Scale Alpha=&FloatFromCurve(FloatCurve=DataInterface() {propertyValues {Curve {keys [
         {arriveTangent=9.0725, interpMode=RCIM_Cubic, leaveTangent=9.0725, tangentMode=RCTM_User, time=0, value=0}
         {arriveTangent=-2.5249, interpMode=RCIM_Cubic, leaveTangent=-0.2678, tangentMode=RCTM_Break, time=0.1, value=1}
         {arriveTangent=-1.6791, interpMode=RCIM_Cubic, leaveTangent=-0.0405, tangentMode=RCTM_Break, time=1, value=0}
        ]}}})
    Scale RGB=&VectorFromFloat(Value=&FloatFromCurve(
      FloatCurve=DataInterface() {propertyValues {Curve {
         defaultValue=1
         keys [
          {arriveTangent=9.0725, interpMode=RCIM_Cubic, leaveTangent=9.0725, tangentMode=RCTM_User, time=0, value=1}
          {arriveTangent=-2.5249, interpMode=RCIM_Cubic, leaveTangent=-2.5249, tangentMode=RCTM_User, time=0.0544, value=2.9665}
         ]
        }}}
      Scale Curve=&MergeHelper_FloatSwitch(0=1.5, 1=2, 2=1.25, 3=1, 4=5, Index=0, Num=::NewEnumerator3)
     ))
    ScaleA=true
    ScaleRGB=true
   )
  ]
  rendererClasses [&NiagaraMeshRendererProperties]
  renderers [
   {rendererClass=&NiagaraMeshRendererProperties, rendererIndex=0}
   {rendererClass=&NiagaraMeshRendererProperties, rendererIndex=1}
   {rendererClass=&NiagaraMeshRendererProperties, rendererIndex=2}
  ]
 }
 LightningLight {
  emitterUpdateScript [
   &EmitterState()
   &SpawnRate(SpawnRate=10, Use Spawn Probability=false)
  ]
  particleSpawnScript [
   &InitializeParticle(
    Color=<rgba 5 5 5 1>
    Lifetime=0.1
    Position Offset=&Add_Vector(A=$System.Offset, B=<0 0 0>)
    UsePositionOffset=true
    Write Material Random=true
   )
   &ShapeLocation(
    Cylinder Distribution=::NewEnumerator2
    Cylinder Height=20
    Cylinder Radius=15
    MaskSpawnGroup=false
    Offset=$System.Offset
    Offset Mode=::NewEnumerator0
    Rotation Angle Type:ENiagara_AngleInput=::NewEnumerator0
    Rotation Mode=::NewEnumerator2
    Shape Primitive=::NewEnumerator1
    U Position=&Sine(Normalized Angle=$Emitter.Age, Period=1.8, Scale=1, Bias=0)
    UseOwnerScale=true
    V Position=&Cosine(Normalized Angle=$Emitter.Age, Period=1.4, Scale=1, Bias=0)
    Yaw / Pitch / Roll=<0 90 0>
   )
  ]
  particleUpdateScript [
   &ParticleState()
   &ScaleColor(
    Scale Alpha=&FloatFromCurve(FloatCurve=DataInterface() {propertyValues {Curve {keys [
         {interpMode=RCIM_Cubic, tangentMode=RCTM_User, time=0, value=0.33}
         {interpMode=RCIM_Cubic, tangentMode=RCTM_User, time=0.5, value=1}
         <at 1 0.33>
        ]}}})
    Scale RGB=<1 0.9 0>
    ScaleA=true
    ScaleRGB=true
   )
  ]
  rendererClasses [&NiagaraComponentRendererProperties]
  renderers [{rendererClass=&NiagaraComponentRendererProperties, rendererIndex=0}]
 }
}
summary {
 systemName=NS_Backpack_WavySignalBanjo_Reactive
 emitters [
  {emitterName=LightningLight, rendererClasses [&NiagaraComponentRendererProperties]}
  {emitterName=Lightning, rendererClasses [&NiagaraMeshRendererProperties]}
  {emitterName=EggGlows, rendererClasses [&NiagaraSpriteRendererProperties]}
  {emitterName=EggRedLight, rendererClasses [&NiagaraComponentRendererProperties]}
 ]
 userVariables [
  User.bIsFrontEnd {type=NiagaraBool}
  User.bIsFrontEndPreview {type=NiagaraBool}
 ]
}
